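# Physics asset for the Ashes female medium hair-down face accessory mesh, setting up physics for its loose hair.
# Five capsule chains (back, left, right, front-left, front-right) run from hold bones linked to the head, with large capsules covering head, neck, upper spine, clavicles and upper arms.
# Hair joints mostly use limited swing and zero twist limits, but the left root is angularly free and lower right joints allow twist.
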PhysicsAsset /BRCosmetics/Characters/Player/Female/Medium/Bodies/F_MED_Ashes_Hair_Down/Meshes/Parts/F_MED_Ashes_Hair_Down_FaceAcc_Physics
bodies {
 clavicle_l [
  {shapeType=Capsule, length=13.6545, radius=6.5311, shapeName=head_capsule, center=<5.8036 0 0>, rotation=<rot -90 -179.9802 179.9803>}
  {shapeType=Capsule, length=95.25, radius=12.25, shapeName=clavicle_l_capsule, center=<24.3806 4.0597 0.0004>}
 ]
 clavicle_r [
  {
   shapeType=Capsule
   length=4.8499
   radius=5.5774
   shapeName=head_capsule
   center=<-5.7359 1.3606 2.1848>
   rotation=<rot -88.6177 179.9989 -179.9989>
  }
  {shapeType=Capsule, length=95.25, radius=12.25, shapeName=clavicle_r_capsule, center=<-27.6054 -13.292 0.0002>}
 ]
 hairBk1_dyn [{shapeType=Capsule, length=10.6147, radius=3.1353, shapeName=hairBk1_dyn_capsule, center=<11.6508 -4.5819 0>, rotation=<rot 90 90 95>}]
 hairBk1_hold [{shapeType=Capsule, length=0.505, radius=0.505, shapeName=hairBk1_hold_capsule, rotation=<rot 0 -90 0>}]
 hairBk2_dyn [{shapeType=Capsule, length=4.5182, radius=3.7318, shapeName=hairLf2_dyn_capsule, center=<9 -4 0>, rotation=<rot 90 0 0>}]
 hairBk3_dyn [{shapeType=Capsule, length=4.75, radius=5.75, shapeName=hairBk3_dyn_capsule, center=<5 -4 0>, rotation=<rot 90 0 0>}]
 hairBk4_dyn [{shapeType=Capsule, length=8, radius=7.75, shapeName=hairBk4_dyn_capsule, center=<5 -7 0>, rotation=<rot 90 0 0>}]
 hairFrLf1_dyn [{shapeType=Capsule, length=2, radius=2, shapeName=hairFrLf1_dyn_capsule, center=<5 0 1>, rotation=<rot 90 0 0>}]
 hairFrLf1_hold [{shapeType=Capsule, length=0.505, radius=0.505, shapeName=hairFrLf1_hold_capsule, rotation=<rot 0 -90 0>}]
 hairFrLf2_dyn [{shapeType=Capsule, length=1, radius=3, shapeName=hairFrLf2_dyn_capsule, center=<5 0 0>, rotation=<rot 90 0 0>}]
 hairFrRt1_dyn [{shapeType=Capsule, length=3.5, radius=3.25, shapeName=hairRt3_front_dyn_capsule, center=<7.2015 2.5001 -4.5>}]
 hairFrRt1_hold [{shapeType=Capsule, length=0.505, radius=0.505, shapeName=hairFrRt1_hold_capsule, rotation=<rot 0 -90 0>}]
 hairFrRt2_dyn [{shapeType=Capsule, length=7.25, radius=2.75, shapeName=hairRt3_front_dyn_capsule, center=<8.8015 3.2001 -2.6>, rotation=<rot 0 0 5.0003>}]
 hairLf1_dyn [{shapeType=Capsule, length=4.5964, radius=4.9036, shapeName=hairRt1_dyn_capsule, center=<10.4801 -4 0>, rotation=<rot 90 0 0>}]
 hairLf1_hold [{shapeType=Capsule, length=2.75, radius=2.005, shapeName=hairLf1_hold_capsule, center=<0 -4 0>, rotation=<rot 0 -90 0>}]
 hairLf2_dyn [{shapeType=Capsule, length=1.25, radius=6.25, shapeName=hairLf3_dyn_capsule, center=<12.457 -4 0>, rotation=<rot 64.9982 -90.0005 -105.0013>}]
 hairLf3_dyn [{
   shapeType=Capsule
   length=10.5
   radius=6.5
   shapeName=hairLf3_dyn_capsule
   center=<15.6863 -3.5 0>
   rotation=<rot 64.9982 -90.0005 -105.0013>
  }]
 hairRt1_dyn [{
   shapeType=Capsule
   length=4
   radius=4.5
   shapeName=hairRt1_dyn_capsule
   center=<-11.149 5.3995 0.0001>
   rotation=<rot 64.9982 -90.0003 -90.0007>
  }]
 hairRt1_hold [{shapeType=Capsule, length=0.505, radius=0.505, shapeName=hairRt1_hold_capsule, rotation=<rot 0 -90 0>}]
 hairRt2_dyn [
  {
   shapeType=Capsule
   length=18.3485
   radius=6.5046
   shapeName=hairLf2_dyn_capsule
   center=<-14.7482 5 -0.4437>
   rotation=<rot 64.9988 -90.0008 -90.001>
  }
  {
   shapeType=Capsule
   length=17.0083
   radius=2.2323
   shapeName=hairRt2_dyn_capsule
   center=<-10.9045 0.5062 -4.1974>
   rotation=<rot 64.9988 -90.0008 -19.6903>
  }
 ]
 hairRt3_dyn [{shapeType=Capsule, length=5.1454, radius=8.7577, shapeName=hairLf3_dyn_capsule, center=<-11.1031 4 0>, rotation=<rot 90 0 0>}]
 head [{shapeType=Capsule, length=4.6545, radius=7.5311, shapeName=head_capsule, center=<1.3572 -0.5172 0.486>, rotation=<rot 9.3054 -82.4056 -90>}]
 neck_02 [{shapeType=Capsule, length=5.25, radius=5.505, shapeName=neck_02_capsule, center=<-3.794 0.0001 -0.0002>, rotation=<rot 0 -89.9999 89.9965>}]
 spine_05 [
  {shapeType=Capsule, length=11.25, radius=9.75, shapeName=spine_05_capsule, center=<-1.2828 0.8147 0>, rotation=<rot 0 -90 0>}
  {shapeType=Capsule, length=16.25, radius=13.75, shapeName=spine_05_capsule, center=<-10.3778 2.2988 0>}
  {shapeType=Capsule, length=36.5, radius=20, shapeName=spine_05_capsule, center=<25.6362 25.8313 0.0002>, rotation=<rot -0.0001 0 0>}
 ]
 upperarm_l [{
   shapeType=Capsule
   length=24
   radius=3.005
   shapeName=upperarm_l_capsule
   center=<14.5947 -0.0001 -0.0016>
   rotation=<rot 0 -89.9992 89.9938>
  }]
 upperarm_r [{
   shapeType=Capsule
   length=22.135
   radius=5.7338
   shapeName=upperarm_r_capsule
   center=<-14.5947 0 -0.0016>
   rotation=<rot 0 -89.9995 89.9938>
  }]
}
constraints [
 {bone1Name=hairFrLf1_dyn, bone2Name=hairFrLf1_hold, swing1LimitDegrees=30, swing2LimitDegrees=30, twistLimitDegrees=0}
 {bone1Name=hairFrLf2_dyn, bone2Name=hairFrLf1_dyn, swing1LimitDegrees=30, swing2LimitDegrees=30, twistLimitDegrees=0}
 {bone1Name=hairRt1_dyn, bone2Name=hairRt1_hold, swing1LimitDegrees=60, swing2LimitDegrees=10, twistLimitDegrees=0}
 {bone1Name=hairRt2_dyn, bone2Name=hairRt1_dyn, swing1LimitDegrees=30, swing2LimitDegrees=45, twistLimitDegrees=30}
 {bone1Name=hairRt3_dyn, bone2Name=hairRt2_dyn, swing1LimitDegrees=30, swing2LimitDegrees=45, twistLimitDegrees=30}
 {bone1Name=hairBk1_dyn, bone2Name=hairBk1_hold, swing1LimitDegrees=60, swing2LimitDegrees=60, twistLimitDegrees=0}
 {bone1Name=hairBk2_dyn, bone2Name=hairBk1_dyn, swing1LimitDegrees=60, swing2LimitDegrees=60, twistLimitDegrees=0}
 {bone1Name=hairBk3_dyn, bone2Name=hairBk2_dyn, swing1LimitDegrees=60, swing2LimitDegrees=60, twistLimitDegrees=0}
 {bone1Name=hairBk4_dyn, bone2Name=hairBk3_dyn, swing1LimitDegrees=60, swing2LimitDegrees=60, twistLimitDegrees=0}
 {
  bone1Name=hairLf1_dyn
  bone2Name=hairLf1_hold
  swing1LimitDegrees=30
  swing1Motion=Free
  swing2LimitDegrees=30
  swing2Motion=Free
  twistLimitDegrees=0
  twistMotion=Free
 }
 {bone1Name=hairLf2_dyn, bone2Name=hairLf1_dyn, swing1LimitDegrees=60, swing2LimitDegrees=60, twistLimitDegrees=0}
 {bone1Name=hairLf3_dyn, bone2Name=hairLf2_dyn, swing1LimitDegrees=60, swing2LimitDegrees=60, twistLimitDegrees=0}
 {
  bone1Name=hairFrLf1_hold
  bone2Name=head
  swing1LimitDegrees=45
  swing1Motion=Free
  swing2LimitDegrees=45
  swing2Motion=Free
  twistLimitDegrees=45
  twistMotion=Free
 }
 {
  bone1Name=hairRt1_hold
  bone2Name=head
  swing1LimitDegrees=45
  swing1Motion=Free
  swing2LimitDegrees=45
  swing2Motion=Free
  twistLimitDegrees=45
  twistMotion=Free
 }
 {
  bone1Name=hairBk1_hold
  bone2Name=head
  swing1LimitDegrees=45
  swing1Motion=Free
  swing2LimitDegrees=45
  swing2Motion=Free
  twistLimitDegrees=45
  twistMotion=Free
 }
 {
  bone1Name=hairLf1_hold
  bone2Name=head
  swing1LimitDegrees=45
  swing1Motion=Free
  swing2LimitDegrees=45
  swing2Motion=Free
  twistLimitDegrees=45
  twistMotion=Free
 }
 {
  bone1Name=clavicle_l
  bone2Name=spine_05
  swing1LimitDegrees=45
  swing1Motion=Free
  swing2LimitDegrees=45
  swing2Motion=Free
  twistLimitDegrees=45
  twistMotion=Free
 }
 {
  bone1Name=clavicle_r
  bone2Name=spine_05
  swing1LimitDegrees=45
  swing1Motion=Free
  swing2LimitDegrees=45
  swing2Motion=Free
  twistLimitDegrees=45
  twistMotion=Free
 }
 {
  bone1Name=head
  bone2Name=neck_02
  swing1LimitDegrees=45
  swing1Motion=Free
  swing2LimitDegrees=45
  swing2Motion=Free
  twistLimitDegrees=45
  twistMotion=Free
 }
 {
  bone1Name=neck_02
  bone2Name=spine_05
  swing1LimitDegrees=45
  swing1Motion=Free
  swing2LimitDegrees=45
  swing2Motion=Free
  twistLimitDegrees=45
  twistMotion=Free
 }
 {bone1Name=hairFrRt1_dyn, bone2Name=hairFrRt1_hold, swing1LimitDegrees=45, swing2LimitDegrees=45, twistLimitDegrees=0}
 {bone1Name=hairFrRt2_dyn, bone2Name=hairFrRt1_dyn, swing1LimitDegrees=45, swing2LimitDegrees=45, twistLimitDegrees=0}
 {
  bone1Name=hairFrRt1_hold
  bone2Name=head
  swing1LimitDegrees=45
  swing1Motion=Free
  swing2LimitDegrees=45
  swing2Motion=Free
  twistLimitDegrees=45
  twistMotion=Free
 }
 {
  bone1Name=upperarm_l
  bone2Name=clavicle_l
  swing1LimitDegrees=45
  swing1Motion=Free
  swing2LimitDegrees=45
  swing2Motion=Free
  twistLimitDegrees=45
  twistMotion=Free
 }
 {
  bone1Name=upperarm_r
  bone2Name=clavicle_r
  swing1LimitDegrees=45
  swing1Motion=Free
  swing2LimitDegrees=45
  swing2Motion=Free
  twistLimitDegrees=45
  twistMotion=Free
 }
]
