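# Post-copy animation Blueprint for the Ashes hair-down face accessory, adding hair physics and bone adjustments onto the pose copied from the attached parent mesh.
# Four RigidBody nodes are gated by skydiving, back bling and cape state, with additive offsets nudging strands and partial constraints tying front-right strands to upper-body bones.
# The event graph derives these alphas from skydiving, targeting and one-time customization checks.
# Several targeting offset and constraint nodes are unconnected.
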AnimBlueprint /BRCosmetics/Characters/Player/Female/Medium/Bodies/F_MED_Ashes_Hair_Down/Meshes/Parts/F_MED_Ashes_Hair_Down_FaceAcc_AnimBP
target_skeleton=&/F_MED_Ashes_Hair_Down_FaceAcc_Skeleton
graphs [
 AnimGraph {nodes [
   n1 = "Constraint - hairFrRt1_dyn (PPPPPP)"!(
    ComponentPose
    Alpha=1
    ConstraintWeights_0=0.033
    ConstraintWeights_1:float=0.5744
    ConstraintWeights_2:float=0.2574
    ConstraintWeights_3:float=0.004
    ConstraintWeights_4:float=0.0641
    ConstraintWeights_5:float=0.0672
   ) -> (Pose) {
    alphaBoolBlend {blendOption=Linear}
    boneToModify {boneName=hairFrRt1_dyn}
    constraintSetup [
     {targetBone {boneName=spine_04}}
     {targetBone {boneName=spine_05}}
     {targetBone {boneName=clavicle_r}}
     {targetBone {boneName=chest_r}}
     {targetBone {boneName=neck_02}}
     {targetBone {boneName=neck_01}}
    ]
    constraintWeights [1,1,1,1,1,1]
   }
   "Constraint - hairFrRt2_dyn (PPPPPPPP)"!(
    ComponentPose=@n1
    Alpha=1
    ConstraintWeights_0=0.0366
    ConstraintWeights_1:float=0.6238
    ConstraintWeights_2:float=0.0081
    ConstraintWeights_3:float=0.0744
    ConstraintWeights_4:float=0.1243
    ConstraintWeights_5:float=0.1279
    ConstraintWeights_6:float=0.0025
    ConstraintWeights_7:float=0.0024
   ) -> (Pose) {
    alphaBoolBlend {blendOption=Linear}
    boneToModify {boneName=hairFrRt2_dyn}
    constraintSetup [
     {targetBone {boneName=spine_04}}
     {targetBone {boneName=spine_05}}
     {targetBone {boneName=upperarm_twist_01_r}}
     {targetBone {boneName=clavicle_r}}
     {targetBone {boneName=neck_02}}
     {targetBone {boneName=neck_01}}
     {targetBone {boneName=head}}
     {targetBone {boneName=C_jaw}}
    ]
    constraintWeights [1,1,1,1,1,1,1,1]
   }
   "Transform (Modify) Bone - Bone: hairRt1_dyn"(Alpha=Get isTargeting() -> (isTargeting:double), Translation="0,3.5,2", Rotation="0,0,0", Scale="1,1,1") {
    rotationMode=BMM_Additive
    rotationSpace=BCS_BoneSpace
    scaleMode=BMM_Additive
    scaleSpace=BCS_BoneSpace
    translationMode=BMM_Additive
    translationSpace=BCS_BoneSpace
    alphaBoolBlend {blendOption=Linear}
    alphaScaleBiasClamp {bInterpResult=true}
    boneToModify {boneName=hairRt1_dyn}
    translation {y=3.5, z=2}
   }
   "Transform (Modify) Bone - Bone: hairRt1_hold"(Alpha=1, Translation="1.369,-0.005,-0.5", Rotation="7.111,0.028,-0.033", Scale="0.977,1.000,1.000") {
    rotationMode=BMM_Additive
    rotationSpace=BCS_BoneSpace
    scaleMode=BMM_Additive
    scaleSpace=BCS_BoneSpace
    translationMode=BMM_Additive
    translationSpace=BCS_BoneSpace
    alphaBoolBlend {blendOption=Linear}
    alphaScaleBiasClamp {bInterpResult=true}
    boneToModify {boneName=hairRt1_hold}
    rotation=<rot 7.111 0.028 -0.033>
    scale {x=0.9773}
    translation=<1.369 -0.005 -0.5>
   }
   "Transform (Modify) Bone - Bone: hairBk4_dyn"(Alpha=1, Translation="0,-5,0", Rotation="0,0,0", Scale="1,1,1") {
    rotationMode=BMM_Additive
    rotationSpace=BCS_BoneSpace
    scaleMode=BMM_Additive
    scaleSpace=BCS_BoneSpace
    translationMode=BMM_Additive
    translationSpace=BCS_BoneSpace
    alphaBoolBlend {blendOption=Linear}
    alphaScaleBiasClamp {bInterpResult=true, interpSpeedDecreasing=1, interpSpeedIncreasing=1}
    boneToModify {boneName=hairBk4_dyn}
    translation {y=-5}
   }
   "Transform (Modify) Bone - Bone: hairRt2_dyn"(Alpha=Get isTargetingNoBling() -> (isTargetingNoBling:double), Translation="0,-3.5,-2", Rotation="0,0,0", Scale="1,1,1") {
    rotationMode=BMM_Additive
    rotationSpace=BCS_BoneSpace
    scaleMode=BMM_Additive
    scaleSpace=BCS_BoneSpace
    translationMode=BMM_Additive
    translationSpace=BCS_BoneSpace
    alphaBoolBlend {blendOption=Linear}
    alphaScaleBiasClamp {bInterpResult=true, interpSpeedDecreasing=1, interpSpeedIncreasing=1}
    boneToModify {boneName=hairRt2_dyn}
    translation {y=-3.5, z=-2}
   }
   n2 = "Transform (Modify) Bone - Bone: hairRt1_dyn"(
    ComponentPose=Local To Component(LocalPose=Copy Pose From Mesh() {bUseAttachedParent=true})
    Alpha=0
    Translation="0,3,3"
    Rotation="-5,0,0"
    Scale="1,1,1"
   ) {
    rotationMode=BMM_Additive
    rotationSpace=BCS_BoneSpace
    scaleMode=BMM_Additive
    scaleSpace=BCS_BoneSpace
    translationMode=BMM_Additive
    translationSpace=BCS_BoneSpace
    alphaBoolBlend {blendOption=Linear}
    alphaScaleBiasClamp {bInterpResult=true}
    boneToModify {boneName=hairRt1_dyn}
    rotation {pitch=-5}
    translation {y=3, z=3}
   }
   n3 = Get Is Sky Diving() -> (Is Sky Diving:double)
   n4 = "Transform (Modify) Bone - Bone: hairRt2_dyn"(ComponentPose=@n2, Alpha=@n3, Translation="0,10,0", Rotation="0,0,0", Scale="1,1,1") {
    rotationMode=BMM_Additive
    rotationSpace=BCS_BoneSpace
    scaleMode=BMM_Additive
    scaleSpace=BCS_BoneSpace
    translationMode=BMM_Additive
    translationSpace=BCS_BoneSpace
    alphaBoolBlend {blendOption=Linear}
    alphaScaleBiasClamp {bInterpResult=true}
    boneToModify {boneName=hairRt2_dyn}
    translation {y=10}
   }
   n5 = "Transform (Modify) Bone - Bone: hairBk1_dyn"(ComponentPose=@n4, Alpha=@n3, Translation="0,-10,0", Rotation="0,0,0", Scale="1,1,1") {
    rotationMode=BMM_Additive
    rotationSpace=BCS_BoneSpace
    scaleMode=BMM_Additive
    scaleSpace=BCS_BoneSpace
    translationMode=BMM_Additive
    translationSpace=BCS_BoneSpace
    alphaBoolBlend {blendOption=Linear}
    alphaScaleBiasClamp {bInterpResult=true}
    boneToModify {boneName=hairBk1_dyn}
    translation {y=-10}
   }
   n6 = "Transform (Modify) Bone - Bone: hairLf1_dyn"(ComponentPose=@n5, Alpha=@n3, Translation="0,-10,0", Rotation="0,0,0", Scale="1,1,1") {
    rotationMode=BMM_Additive
    rotationSpace=BCS_BoneSpace
    scaleMode=BMM_Additive
    scaleSpace=BCS_BoneSpace
    translationMode=BMM_Additive
    translationSpace=BCS_BoneSpace
    alphaBoolBlend {blendOption=Linear}
    alphaScaleBiasClamp {bInterpResult=true}
    boneToModify {boneName=hairLf1_dyn}
    translation {y=-10}
   }
   n7 = RigidBody(ComponentPose=@n6, Alpha=Get Is Sky Diving()) {
    bClampLinearTranslationLimitToRefPose=true
    bOverrideWorldGravity=true
    overridePhysicsAsset=&/F_MED_Ashes_Hair_Down_FaceAcc_PhysicsSkydive2
    simulationSpace=BaseBoneSpace
    alphaBoolBlend {blendOption=Linear}
    baseBoneRef {boneName=head}
    componentAppliedLinearAccClamp=<3000 3000 3000>
    componentLinearAccScale=<0.4 0.3 0.3>
    componentLinearVelScale=<0.4 0.3 0.3>
    overrideWorldGravity {z=-980}
   }
   n8 = NOT Boolean(A=Get isWearingBackbling())
   n9 = NOT Boolean(A=Get bIsSkyDiving())
   n10 = RigidBody(ComponentPose=@n7, Alpha="To Float (Boolean)"(InBool=AND Boolean(A=@n8, B=@n9))) {
    bClampLinearTranslationLimitToRefPose=true
    bOverrideWorldGravity=true
    overridePhysicsAsset=&/F_MED_Ashes_Hair_Down_FaceAcc_Physics
    simulationSpace=BaseBoneSpace
    alphaBoolBlend {blendOption=Linear}
    baseBoneRef {boneName=neck_02}
    componentAppliedLinearAccClamp=<3000 3000 3000>
    componentLinearAccScale=<0.4 0.3 0.3>
    componentLinearVelScale=<0.4 0.3 0.3>
    overrideWorldGravity {z=-980}
    simSpaceSettings {
     maxAngularAcceleration=1000
     maxAngularVelocity=3
     maxLinearAcceleration=700
     maxLinearVelocity=500
     velocityScaleZ=0
     worldAlpha=0.2
     externalLinearDragV=<0.2 0.2>
    }
   }
   n11 = NOT Boolean(A=Get bIsSkyDiving())
   n12 = NOT Boolean(A=Get isWearingCapeBool() -> (isWearingCapeBool:bool))
   n13 = Get isWearingBackbling()
   n14 = RigidBody(ComponentPose=@n10, Alpha="To Float (Boolean)"(InBool=AND Boolean!(A=@n13, B=@n12, C:bool=@n11) -> (ReturnValue))) {
    bClampLinearTranslationLimitToRefPose=true
    bOverrideWorldGravity=true
    overridePhysicsAsset=&/F_MED_Ashes_Hair_Down_FaceAcc_PhysicsBackBling3
    simulationSpace=BaseBoneSpace
    alphaBoolBlend {blendOption=Linear}
    baseBoneRef {boneName=head}
    componentAppliedLinearAccClamp=<3000 3000 3000>
    componentLinearAccScale=<0.4 0.3 0.3>
    componentLinearVelScale=<0.4 0.3 0.3>
    overrideWorldGravity {z=-980}
   }
   n15 = NOT Boolean(A=Get bIsSkyDiving())
   n16 = Get isWearingCapeBool()
   n17 = RigidBody(ComponentPose=@n14, Alpha="To Float (Boolean)"(InBool=AND Boolean!(A=@n16, B=@n15) -> (ReturnValue))) {
    bClampLinearTranslationLimitToRefPose=true
    bOverrideWorldGravity=true
    overridePhysicsAsset=&F_MED_Ashes_Hair_Down_FaceAcc_Physics
    simulationSpace=BaseBoneSpace
    alphaBoolBlend {blendOption=Linear}
    baseBoneRef {boneName=head}
    componentAppliedLinearAccClamp=<3000 3000 3000>
    componentLinearAccScale=<0.4 0.3 0.3>
    componentLinearVelScale=<0.4 0.3 0.3>
    overrideWorldGravity {z=-980}
   }
   n18 = "Transform (Modify) Bone - Bone: hairRt1_hold"(ComponentPose=@n17, Alpha=1, Translation="0,0,0", Rotation="7.111,0.028,-0.033", Scale="0.977,1.000,1.000") {
    rotationMode=BMM_Additive
    rotationSpace=BCS_BoneSpace
    scaleMode=BMM_Additive
    scaleSpace=BCS_BoneSpace
    translationMode=BMM_Additive
    translationSpace=BCS_BoneSpace
    alphaBoolBlend {blendOption=Linear}
    alphaScaleBiasClamp {bInterpResult=true}
    boneToModify {boneName=hairRt1_hold}
    rotation=<rot 7.111 0.028 -0.033>
    scale {x=0.9773}
   }
   n19 = "Transform (Modify) Bone - Bone: hairLf2_dyn"(ComponentPose=@n18, Alpha=Get Is Sky Diving(), Translation="0,0,-2", Rotation="0,0,0", Scale="1,1,1") {
    rotationMode=BMM_Additive
    rotationSpace=BCS_BoneSpace
    scaleMode=BMM_Additive
    scaleSpace=BCS_BoneSpace
    translationMode=BMM_Additive
    translationSpace=BCS_BoneSpace
    alphaBoolBlend {blendOption=Linear}
    alphaScaleBiasClamp {bInterpResult=true, interpSpeedDecreasing=1, interpSpeedIncreasing=1}
    boneToModify {boneName=hairLf2_dyn}
    translation {z=-2}
   }
   n20 = "Transform (Modify) Bone - Bone: hairRt3_dyn"(ComponentPose=@n19, Alpha=Get backBlingTargeting() -> (backBlingTargeting:double), Translation="0,1,2", Rotation="30,0,0", Scale="1,1,1") {
    rotationMode=BMM_Additive
    rotationSpace=BCS_BoneSpace
    scaleMode=BMM_Additive
    scaleSpace=BCS_BoneSpace
    translationMode=BMM_Additive
    translationSpace=BCS_BoneSpace
    alphaBoolBlend {blendOption=Linear}
    alphaScaleBiasClamp {bInterpResult=true, interpSpeedDecreasing=1, interpSpeedIncreasing=1}
    boneToModify {boneName=hairRt3_dyn}
    rotation {pitch=30}
    translation {y=1, z=2}
   }
   n21 = Get isWearingCape() -> (isWearingCape:double)
   n22 = "Transform (Modify) Bone - Bone: hairRt3_dyn"(ComponentPose=@n20, Alpha=@n21, Translation="0,5,3", Rotation="0,0,0", Scale="1,1,1") {
    rotationMode=BMM_Additive
    rotationSpace=BCS_BoneSpace
    scaleMode=BMM_Additive
    scaleSpace=BCS_BoneSpace
    translationMode=BMM_Additive
    translationSpace=BCS_BoneSpace
    alphaBoolBlend {blendOption=Linear}
    alphaScaleBiasClamp {bInterpResult=true, interpSpeedDecreasing=1, interpSpeedIncreasing=1}
    boneToModify {boneName=hairRt3_dyn}
    translation {y=5, z=3}
   }
   n23 = "Transform (Modify) Bone - Bone: hairBk2_dyn"(ComponentPose=@n22, Alpha=1, Translation="0,-3.5,0", Rotation="0,0,0", Scale="1,1,1") {
    rotationMode=BMM_Additive
    rotationSpace=BCS_BoneSpace
    scaleMode=BMM_Additive
    scaleSpace=BCS_BoneSpace
    translationMode=BMM_Additive
    translationSpace=BCS_BoneSpace
    alphaBoolBlend {blendOption=Linear}
    alphaScaleBiasClamp {bInterpResult=true, interpSpeedDecreasing=1, interpSpeedIncreasing=1}
    boneToModify {boneName=hairBk2_dyn}
    translation {y=-3.5}
   }
   n24 = "Transform (Modify) Bone - Bone: hairBk3_dyn"(ComponentPose=@n23, Alpha=@n21, Translation="0,-4,0", Rotation="0,0,0", Scale="1,1,1") {
    rotationMode=BMM_Additive
    rotationSpace=BCS_BoneSpace
    scaleMode=BMM_Additive
    scaleSpace=BCS_BoneSpace
    translationMode=BMM_Additive
    translationSpace=BCS_BoneSpace
    alphaBoolBlend {blendOption=Linear}
    alphaScaleBiasClamp {bInterpResult=true, interpSpeedDecreasing=1, interpSpeedIncreasing=1}
    boneToModify {boneName=hairBk3_dyn}
    translation {y=-4}
   }
   n25 = "Transform (Modify) Bone - Bone: hairLf3_dyn"(ComponentPose=@n24, Alpha=@n21, Translation="0,-5,-3", Rotation="0,0,0", Scale="1,1,1") {
    rotationMode=BMM_Additive
    rotationSpace=BCS_BoneSpace
    scaleMode=BMM_Additive
    scaleSpace=BCS_BoneSpace
    translationMode=BMM_Additive
    translationSpace=BCS_BoneSpace
    alphaBoolBlend {blendOption=Linear}
    alphaScaleBiasClamp {bInterpResult=true, interpSpeedDecreasing=1, interpSpeedIncreasing=1}
    boneToModify {boneName=hairLf3_dyn}
    translation {y=-5, z=-3}
   }
   n26 = "Constraint - hairFrRt1_dyn (PPPPP)"!(
    ComponentPose=@n25
    Alpha=0.6
    ConstraintWeights_0=0.033
    ConstraintWeights_1:float=0.5784
    ConstraintWeights_2:float=0.2574
    ConstraintWeights_3:float=0.0641
    ConstraintWeights_4:float=0.0672
   ) -> (Pose) {
    alphaBoolBlend {blendOption=Linear}
    boneToModify {boneName=hairFrRt1_dyn}
    constraintSetup [
     {targetBone {boneName=spine_04}}
     {targetBone {boneName=spine_05}}
     {targetBone {boneName=clavicle_r}}
     {targetBone {boneName=neck_02}}
     {targetBone {boneName=neck_01}}
    ]
    constraintWeights [1,1,1,1,1]
   }
   n27 = "Constraint - hairFrRt2_dyn (PPPPP)"!(
    ComponentPose=@n26
    Alpha=0.8
    ConstraintWeights_0=0.0366
    ConstraintWeights_1:float=0.6238
    ConstraintWeights_2:float=0.0744
    ConstraintWeights_3:float=0.1243
    ConstraintWeights_4:float=0.1279
   ) -> (Pose) {
    alphaBoolBlend {blendOption=Linear}
    boneToModify {boneName=hairFrRt2_dyn}
    constraintSetup [
     {targetBone {boneName=spine_04}}
     {targetBone {boneName=spine_05}}
     {targetBone {boneName=clavicle_r}}
     {targetBone {boneName=neck_02}}
     {targetBone {boneName=neck_01}}
    ]
    constraintWeights [1,1,1,1,1]
   }
   Output Pose(Result=Component To Local(ComponentPose=@n27)) {layerGroup=None}
  ]}
 EventGraph {nodes [
   Event BlueprintUpdateAnimation()
   n1 = NOT Boolean(A=Get isWearingCapeBool())
   n2 = Get bIsSkyDiving()
   > Set isTargetingNoBling(isTargetingNoBling:double="To Float (Boolean)"(InBool=AND Boolean(A=@n1, B=@n2))) -> (Output_Get:double)
   > Set Is Sky Diving(Is Sky Diving:double="To Float (Boolean)"(InBool=@n2)) -> (Output_Get:double)
   > Delay(Duration=0.2)
   > Set Is Sky Diving(Is Sky Diving=0)
   n3 = Get bIsTargeting()
   n4 = Get isWearingBackbling()
   > Set backBlingTargeting(backBlingTargeting:double="To Float (Boolean)"(InBool=AND Boolean(A=@n3, B=@n4))) -> (Output_Get:double)
   > n5 = Branch(Condition=Get bIsTargeting())
   > Set isTargeting(isTargeting:double=1) -> (Output_Get:double)
   Set isTargeting(execute=@n5.else, isTargeting=0)
   Event OnCharacterCustomizationLoaded()
   > Do Once()
   n6 = CheckForBackbling()
   > Set isWearingBackbling(isWearingBackbling=OR Boolean(A=@n6, B=@n6.bIsBackblingLong))
   > Do Once()
   > Set isWearingCapeBool(isWearingCapeBool:bool=CheckForBackbling().bIsBackblingACape) -> (Output_Get:bool)
   > Do Once()
   > n7 = Branch(Condition=CheckForBackbling().bIsBackblingACape)
   > Set isWearingCape(isWearingCape:double=1) -> (Output_Get:double)
   Set isWearingCape(execute=@n7.else, isWearingCape=0)
   Set isSkydivingInitial(isSkydivingInitial:double=1) -> (Output_Get:double)
   > Delay(Duration=0.2)
   > Set isSkydivingInitial(isSkydivingInitial=0)
   Break Athena Emergency Notice!(AthenaEmergencyNotice:"Athena Emergency Notice Structure (by ref)") -> (Title:String, Body:String) {is=title K2Node_BreakStruct}
   Event OnBeginSkydiving()
   Get isWearingBackbling()
  ]}
]
variables [
 backBlingTargeting {type=number}
 Is Sky Diving {type=number}
 isSkydivingInitial {type=number}
 isTargeting {type=number}
 isTargetingNoBling {type=number}
 isWearingBackbling {type=boolean}
 isWearingCape {type=number}
 isWearingCapeBool {type=boolean}
 NewVar_0 {type=number}
]
